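; Day sequence volume Blueprint for the Sparks lobby's Festival S05 Greenroom environment, switching its day sequence volume on or off.
; On BeginPlay it runs the parent logic, then passes UpdateDaySequenceVolume as a delegate to BindScalabilitySettings.
; That function enables or disables the volume according to the UseVolume variable.
; The overlap and tick events are empty, and the construction script only calls the parent.
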(event EventBeginPlay
  (|CallParentFunction)
  (Default|BindScalabilitySettings (EventDispatchers|CreateEvent :!event UpdateDaySequenceVolume)))

(event Collision|EventActorBeginOverlap
  (params (OtherActor :type Actor%20Object%20Reference)))

(event EventTick
  (params (DeltaSeconds :type Float%20%28single-precision%29)))

(fn UpdateDaySequenceVolume
  (if (Variables|Default|GetUseVolume)
    (CallFunction|EnableDaysequenceVolume)
    (else
      (CallFunction|DisableDaysequenceVolume))))

(fn ConstructionScript
  (|CallParentFunction))
